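# Additive unlit Niagara sprite and mesh material that renders a shaking, flickering chromatic aberration of a texture for the 3DHB hologram FX.
# Three alpha samples at offset UVs, optionally inverted by IsFrom3D, become RGB emissive; each offset uses only its U parameter on both axes.
# Time-based shake and particle-randomized MF_3DHB_HologramFX flicker drive the offsets.
# Opacity favors fringes where channels differ, scaled by flicker and particle alpha.
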Material /Game/Tools/3DHB/FX/PresetFX/Aberration/M_3DHB_FX_Aberration
graph {
 properties {
  bUsedWithNiagaraMeshParticles=true
  bUsedWithNiagaraSprites=true
  blendMode=BLEND_Additive
  pixelDepthOffsetMode=PDOM_Legacy
  shadingModel=MSM_Unlit
  usedShadingModels=Unlit
 }
 nodes [
  n1 = TextureCoordinate()
  n2 = Add(A=Multiply(A=Add(A=@n1) {constB=-0.5}, B=Constant() {r=2})) {constB=0.5}
  n3 = ScalarParameter() {defaultValue=100, parameterName=Shake_Frequency}
  n4 = Time() {bOverride_Period=true, period=6000}
  n5 = LinearInterpolate(A=ScalarParameter() {defaultValue=1, parameterName=Shake_LowestStrength}, Alpha=Frac(None=Multiply(A=@n3, B=@n4))) {constA=0.5}
  n6 = Add(A=ScalarParameter() {defaultValue=4, parameterName=AberrationStrength})
  n7 = MaterialFunctionCall!(
   Time=Add(A=@n4, B=DynamicParameter() {paramNames [Random,Param2,Param3,Param4]}.Random)
   MeshScaleIn
   UVForScanline
   CustomFlicker
   UseCustomFlicker
   UVForStroke
   UVForGlitch
   BaseUVIn
   EmissiveIn
   OpacityIn
   WPOIn
  ) {materialFunction=/Game/Tools/3DHB/Shader/Base/MF_3DHB_HologramFX}
  n8 = OneMinus(None=LinearInterpolate(B=Multiply(A=@n5, B=@n6), Alpha=@n7.UseForFX_FlickerOut) {constA=1, constB=0.2})
  n9 = VectorParameter() {parameterName="Offset_A&&B", channelNames {a=B_V, b=B_U, g=A_V, r=A_U}, defaultValue=<rgba -0.01 0 0.01 0>}
  n10 = AppendVector(A=@n9.A_U, B=@n9.A_U)
  n11 = TextureObjectParameter() {parameterName=Texture, texture=/Game/Tools/3DHB/DefaultAssets/T_3DHB_DefaultTexture}
  n12 = TextureSample(UVs=Add(A=@n2, B=Multiply(A=@n8, B=@n10)), Tex=@n11) {samplerSource=SSM_Clamp_WorldGroupSettings}
  n13 = ScalarParameter() {parameterName=IsFrom3D}
  n14 = LinearInterpolate(A=@n12.A, B=OneMinus(None=@n12.A), Alpha=@n13)
  n15 = VectorParameter() {parameterName=EveryChannelStrength, defaultValue=<rgba 1 1 1 0>}
  n16 = Multiply(A=@n14, B=@n15.R)
  n17 = VectorParameter() {parameterName=Offset_C, channelNames {a="", b="", g=C_V, r=C_U}}
  n18 = AppendVector(A=@n17.C_U, B=@n17.C_U)
  n19 = TextureSample(UVs=Add(A=@n2, B=Multiply(A=@n18, B=@n8)), Tex=@n11) {samplerSource=SSM_Clamp_WorldGroupSettings}
  n20 = LinearInterpolate(A=@n19.A, B=OneMinus(None=@n19.A), Alpha=@n13)
  n21 = Multiply(A=@n20, B=@n15.G)
  n22 = AppendVector(A=@n9.B_U, B=@n9.B_U)
  n23 = TextureSample(UVs=Add(A=@n2, B=Multiply(A=@n22, B=@n8)), Tex=@n11) {samplerSource=SSM_Clamp_WorldGroupSettings}
  n24 = LinearInterpolate(A=@n23.A, B=OneMinus(None=@n23.A), Alpha=@n13)
  MaterialFunctionCall(Emissive=AppendVector(A=AppendVector(A=@n16, B=@n21), B=Multiply(A=@n24, B=@n15.B))) {materialFunction=&MF_FX_EmissiveAdjustments}
  n25 = Min(A=Min(A=@n14, B=@n20), B=@n24)
  n26 = Multiply(A=Max(A=Max(A=@n14, B=@n20), B=@n24), B=Power(Base=OneMinus(None=@n25)) {constExponent=1.8})
  n27 = ScalarParameter() {defaultValue=1, parameterName=Opacity}
  Multiply(A=Multiply(A=@n7.UseForFX_FlickerOut, B=Multiply(A=@n26, B=@n27)), B=ParticleColor().A)
 ]
}
